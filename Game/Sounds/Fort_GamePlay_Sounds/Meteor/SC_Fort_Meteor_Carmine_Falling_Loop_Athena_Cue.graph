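# Sound cue for the falling Carmine meteor in Athena, routed to the gadgets LTM sound class.
# A root mixer blends three separately attenuated layers: looping close and Shepard waves, a very-distant loop, and a non-looping incoming wave.
# Close and distant layers use dedicated attenuation assets, while the incoming wave overrides falloff inline.
# The distant mixer has an empty first input, and the cue references a meteor fall-impact concurrency set.
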SoundCue /Game/Sounds/Fort_GamePlay_Sounds/Meteor/SC_Fort_Meteor_Carmine_Falling_Loop_Athena_Cue
root=Mixer_2
nodes [
 {
  id=Attenuation_0
  children [Mixer_0]
  properties {attenuationSettings=&/Fort_Meteor_Carmine_Athena_Att}
 }
 {
  id=Attenuation_1
  children [WavePlayer_1]
  properties {bOverrideAttenuation=true, attenuationOverrides {falloffDistance=1500, stereoSpread=1000, attenuationShapeExtents {x=3000}}}
 }
 {
  id=Attenuation_2
  children [Mixer_4]
  properties {attenuationSettings=&/Fort_Meteor_Carmine_Distant_Athena_Att}
 }
 {id=Mixer_0, children [WavePlayer_4,WavePlayer_2], properties {inputVolume [3,1]}}
 {id=Mixer_2, children [Attenuation_0, Attenuation_2, Attenuation_1], properties {inputVolume [1.5,0.4,0.75]}}
 {id=Mixer_4, children [null,WavePlayer_0], properties {inputVolume [3,2]}}
 {id=WavePlayer_0, properties {bLooping=true, soundWaveAssetPtr=&/Meteor_Falling_VeryDistant_lp_01a}}
 {id=WavePlayer_1, properties {soundWaveAssetPtr=&/Meteor_Carmine_Incoming_04}}
 {id=WavePlayer_2, properties {bLooping=true, soundWaveAssetPtr=&/Meteor_Carmine_FallingLoop_Close_06}}
 {id=WavePlayer_4, properties {bLooping=true, soundWaveAssetPtr=&/Meteor_Carmine_FallingLoop_Shepard}}
]
properties {
 maxDistance=2097152
 soundClassObject=/Game/Sounds/FortSoundClassesAndModes/Fort_Gadgets_LTM
 volumeMultiplier=0.4
 attenuationOverrides {falloffDistance=1000000, attenuationShapeExtents {x=0}}
 concurrencyOverrides {maxCount=2, resolutionRule=StopQuietest}
 concurrencySet [&/Meteor_Rock_FallImpact_Conc]
}
